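# Sound cue for Slone's idle voice lines as an NPC, routed to the NPC VO sound class.
# A Random node picks among five equally weighted idle wave players and is excluded from branch culling.
# The cue plays at reduced volume, with Slone-specific idle attenuation limiting its range and a Slone chatter concurrency set governing overlapping playback.
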SoundCue /Slone/NPCs/Slone/Slone/Audio/VO/Slone_Idle_Cue
root=Random_0
nodes [
 {
  id=Random_0
  children [WavePlayer_2,WavePlayer_3,WavePlayer_4,WavePlayer_5,WavePlayer_6]
  properties {bShouldExcludeFromBranchCulling=true, weights [1,1,1,1,1]}
 }
 {id=WavePlayer_2, properties {soundWaveAssetPtr=/Slone/NPCs/Slone/Slone/Audio/VO/Waves/Slone_2021_Idle_010}}
 {id=WavePlayer_3, properties {soundWaveAssetPtr=/Slone/NPCs/Slone/Slone/Audio/VO/Waves/Slone_2021_Idle_020}}
 {id=WavePlayer_4, properties {soundWaveAssetPtr=/Slone/NPCs/Slone/Slone/Audio/VO/Waves/Slone_2021_Idle_030}}
 {id=WavePlayer_5, properties {soundWaveAssetPtr=/Slone/NPCs/Slone/Slone/Audio/VO/Waves/Slone_2021_Idle_040}}
 {id=WavePlayer_6, properties {soundWaveAssetPtr=/Slone/NPCs/Slone/Slone/Audio/VO/Waves/Slone_2021_Idle_050}}
]
properties {
 attenuationSettings=&/Slone_Idle_Att
 maxDistance=4000
 soundClassObject=&Fort_VO_NPC
 volumeMultiplier=0.7
 concurrencySet [&/Slone_Chatter_Conc]
}
