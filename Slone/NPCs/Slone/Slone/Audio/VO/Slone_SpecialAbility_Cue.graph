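# Sound cue for the Slone NPC's special ability voice lines.
# A Random node, excluded from branch culling, picks among four equally weighted voice waves.
# Playback uses the NPC VO sound class, a Slone close special-ability attenuation and a chatter concurrency group, with a slight volume boost.
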SoundCue /Slone/NPCs/Slone/Slone/Audio/VO/Slone_SpecialAbility_Cue
root=Random_1
nodes [
 {
  id=Random_1
  children [WavePlayer_0,WavePlayer_6,WavePlayer_7,WavePlayer_8]
  properties {bShouldExcludeFromBranchCulling=true, weights [1,1,1,1]}
 }
 {id=WavePlayer_0, properties {soundWaveAssetPtr=/Slone/NPCs/Slone/Slone/Audio/VO/Waves/Slone_2021_SpecialAbility_010}}
 {id=WavePlayer_6, properties {soundWaveAssetPtr=/Slone/NPCs/Slone/Slone/Audio/VO/Waves/Slone_2021_SpecialAbility_011}}
 {id=WavePlayer_7, properties {soundWaveAssetPtr=/Slone/NPCs/Slone/Slone/Audio/VO/Waves/Slone_2021_SpecialAbility_012}}
 {id=WavePlayer_8, properties {soundWaveAssetPtr=/Slone/NPCs/Slone/Slone/Audio/VO/Waves/Slone_2021_SpecialAbility_013}}
]
properties {
 attenuationSettings=&/Slone_Close_SpecialAbility_Att
 maxDistance=4000
 soundClassObject=&Fort_VO_NPC
 volumeMultiplier=1.2
 concurrencySet [&/Slone_Chatter_Conc]
}
